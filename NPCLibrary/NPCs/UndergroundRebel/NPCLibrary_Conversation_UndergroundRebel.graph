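# Conversation database for the UndergroundRebel NPC in the NPCLibrary, registered under the tag Conversation.Entry.UndergroundRebel.
# It holds a single entry point node, which that tag resolves to.
# The node has no outputs, so the graph contains no dialogue, choices or tasks.
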ConversationDatabase /NPCLibrary/NPCs/UndergroundRebel/NPCLibrary_Conversation_UndergroundRebel
entry_points [{tag=Conversation.Entry.UndergroundRebel, destinations [#1]}]
nodes [ConversationEntryPointNode_7 {
  id=#1
  class=ConversationEntryPointNode
  outputs []
  properties {entryTag {tagName=Conversation.Entry.UndergroundRebel}}
 }]
