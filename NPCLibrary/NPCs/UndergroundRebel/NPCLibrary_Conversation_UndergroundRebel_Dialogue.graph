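# Conversation database holding the "Ask About..." lore dialogue for the Underground Rebel NPC.
# The tagged entry point opens a hub line that branches into three topics (Vengeance, Dark Voyager, The Seven) plus a Back option.
# Each topic is a linear chain of Speech nodes advanced by "Continue..." choices and ends in a ReturnToStart node labelled BACK.
# It contains only dialogue and navigation, with no requirement or side-effect sub-nodes.
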ConversationDatabase /NPCLibrary/NPCs/UndergroundRebel/NPCLibrary_Conversation_UndergroundRebel_Dialogue
entry_points [{tag=Conversation.Entry.UndergroundRebel.Dialogue, destinations [#1]}]
nodes [
 ConversationEntryPointNode_0 {
  id=#1
  class=ConversationEntryPointNode
  outputs [#2]
  properties {entryTag {tagName=Conversation.Entry.UndergroundRebel.Dialogue}}
 }
 Speech_C_12 {
  id=#2
  class=Speech_C
  outputs [#3, #4, #5, #6]
  properties {nodeName="", softTaskIcon=&T_UI_NPC_question, generalConfig {defaultMessage=What d'you want?}}
  sub_nodes [ConversationChoiceNode_39 {class=ConversationChoiceNode, properties {defaultChoiceDisplayText=Ask About...}}]
 }
 Speech_C_20 {
  id=#6
  class=Speech_C
  outputs [#7]
  properties {
   nodeName=""
   softTaskIcon=&T_UI_NPC_question
   generalConfig {defaultMessage="The best allies you could ask for. And we've got two on our side now!"}
  }
  sub_nodes [ConversationChoiceNode_51 {class=ConversationChoiceNode, properties {defaultChoiceDisplayText=The Seven}}]
 }
 Speech_C_21 {
  id=#7
  class=Speech_C
  outputs [#8]
  properties {
   nodeName=""
   softTaskIcon=&T_UI_NPC_question
   generalConfig {defaultMessage="Sure, The Order's pretty quiet... and The Visitor's an odd duck."}
  }
  sub_nodes [ConversationChoiceNode_53 {class=ConversationChoiceNode, properties {defaultChoiceDisplayText=Continue...}}]
 }
 Speech_C_22 {
  id=#8
  class=Speech_C
  outputs [#9]
  properties {
   nodeName=""
   softTaskIcon=&T_UI_NPC_question
   generalConfig {defaultMessage=But we're going to need all the help we can get.}
  }
  sub_nodes [ConversationChoiceNode_54 {class=ConversationChoiceNode, properties {defaultChoiceDisplayText=Continue...}}]
 }
 Speech_C_23 {
  id=#9
  class=Speech_C
  outputs [#10]
  properties {
   nodeName=""
   softTaskIcon=&T_UI_NPC_question
   generalConfig {defaultMessage="I'm hoping they're not the last of The Seven to join us, though..."}
  }
  sub_nodes [ConversationChoiceNode_55 {class=ConversationChoiceNode, properties {defaultChoiceDisplayText=Continue...}}]
 }
 ReturnToStart_C_4 {
  id=#10
  class=ReturnToStart_C
  properties {softTaskIcon=&T_UI_NPC_Back}
  sub_nodes [ConversationChoiceNode_52 {class=ConversationChoiceNode, properties {defaultChoiceDisplayText=BACK}}]
 }
 Speech_C_17 {
  id=#5
  class=Speech_C
  outputs [#11]
  properties {
   nodeName=""
   softTaskIcon=&T_UI_NPC_question
   generalConfig {defaultMessage="Whatever he is, he's collecting Zero Point Shards for a purpose."}
  }
  sub_nodes [ConversationChoiceNode_46 {class=ConversationChoiceNode, properties {defaultChoiceDisplayText=Dark Voyager}}]
 }
 Speech_C_18 {
  id=#11
  class=Speech_C
  outputs [#12]
  properties {
   nodeName=""
   softTaskIcon=&T_UI_NPC_question
   generalConfig {defaultMessage=And I don't think it's for happiness and world peace.}
  }
  sub_nodes [ConversationChoiceNode_48 {class=ConversationChoiceNode, properties {defaultChoiceDisplayText=Continue...}}]
 }
 Speech_C_19 {
  id=#12
  class=Speech_C
  outputs [#13]
  properties {
   nodeName=""
   softTaskIcon=&T_UI_NPC_question
   generalConfig {defaultMessage="Keep your eyes peeled, but hold tight and don't move on him."}
  }
  sub_nodes [ConversationChoiceNode_15 {class=ConversationChoiceNode, properties {defaultChoiceDisplayText=Continue...}}]
 }
 Speech_C_24 {
  id=#13
  class=Speech_C
  outputs [#14]
  properties {
   nodeName=""
   softTaskIcon=&T_UI_NPC_question
   generalConfig {defaultMessage=We can't take him in a fair fight. Not yet.}
  }
  sub_nodes [ConversationChoiceNode_49 {class=ConversationChoiceNode, properties {defaultChoiceDisplayText=Continue...}}]
 }
 ReturnToStart_C_3 {
  id=#14
  class=ReturnToStart_C
  properties {softTaskIcon=&T_UI_NPC_Back}
  sub_nodes [ConversationChoiceNode_47 {class=ConversationChoiceNode, properties {defaultChoiceDisplayText=BACK}}]
 }
 Back_C_2 {
  id=#4
  class=Back_C
  sub_nodes [ConversationChoiceNode_40 {class=ConversationChoiceNode, properties {defaultChoiceDisplayText=BACK}}]
 }
 Speech_C_13 {
  id=#3
  class=Speech_C
  outputs [#15]
  properties {
   nodeName=""
   softTaskIcon=&T_UI_NPC_question
   generalConfig {defaultMessage=I took up vengeance after my buddy Peely got banana-napped... but it goes a little deeper than that.}
  }
  sub_nodes [ConversationChoiceNode_41 {class=ConversationChoiceNode, properties {defaultChoiceDisplayText=Vengeance}}]
 }
 Speech_C_14 {
  id=#15
  class=Speech_C
  outputs [#16]
  properties {
   nodeName=""
   softTaskIcon=&T_UI_NPC_question
   generalConfig {defaultMessage="A while ago, faced with an overwhelming evil force... I didn't fight. To save everything, and I mean EVERYTHING, I ran."}
  }
  sub_nodes [ConversationChoiceNode_14 {class=ConversationChoiceNode, properties {defaultChoiceDisplayText=Continue...}}]
 }
 Speech_C_15 {
  id=#16
  class=Speech_C
  outputs [#17]
  properties {
   nodeName=""
   softTaskIcon=&T_UI_NPC_question
   generalConfig {defaultMessage="I thought that was it. I lost my chance, but I can feel it on the wind..."}
  }
  sub_nodes [ConversationChoiceNode_44 {class=ConversationChoiceNode, properties {defaultChoiceDisplayText=Continue...}}]
 }
 Speech_C_16 {
  id=#17
  class=Speech_C
  outputs [#18]
  properties {
   nodeName=""
   softTaskIcon=&T_UI_NPC_question
   generalConfig {defaultMessage=Maybe I'll get a second shot at the vengeance I deserve.}
  }
  sub_nodes [ConversationChoiceNode_45 {class=ConversationChoiceNode, properties {defaultChoiceDisplayText=Continue...}}]
 }
 ReturnToStart_C_2 {
  id=#18
  class=ReturnToStart_C
  properties {softTaskIcon=&T_UI_NPC_Back}
  sub_nodes [ConversationChoiceNode_42 {class=ConversationChoiceNode, properties {defaultChoiceDisplayText=BACK}}]
 }
]
